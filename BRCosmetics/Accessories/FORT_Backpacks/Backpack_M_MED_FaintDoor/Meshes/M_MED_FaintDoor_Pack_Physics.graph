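# Physics asset for the FaintDoor backpack cosmetic, giving its pack and dangling left, center and right charm chains bodies for physical motion.
# Every body is a single capsule, linked from root through pack to the three charm bases and down each chain.
# Chain links allow moderate swing with very tight twist, as does the center base, while the pack and side bases rotate freely.
# The center chain is shorter, with two dynamic links against four per side.
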PhysicsAsset /BRCosmetics/Accessories/FORT_Backpacks/Backpack_M_MED_FaintDoor/Meshes/M_MED_FaintDoor_Pack_Physics
bodies {
 c_base_charm [{shapeType=Capsule, length=1.1013, radius=1.6823, shapeName=c_base_charm_capsule, center=<0.6872 0 0>, rotation=<rot 0 0 90>}]
 c_dyn_charm_2 [{shapeType=Capsule, length=1, radius=1, center=<1 0 0>, rotation=<rot 90 -90 0>}]
 c_dyn_charm_3 [{shapeType=Capsule, length=1, radius=1, center=<1 0 0>, rotation=<rot 90 -90 0>}]
 l_base_charm [{shapeType=Capsule, length=10.2148, radius=5.1075, shapeName=l_base_charm_capsule, center=<5.1661 -3.2716 -6.8889>, rotation=<rot 0 0 90>}]
 l_dyn_charm_1 [{shapeType=Capsule, length=1, radius=1, center=<2.5176 0 0>, rotation=<rot 90 0 0>}]
 l_dyn_charm_2 [{shapeType=Capsule, length=1, radius=1, center=<1.355 0 0>, rotation=<rot 90 0 0>}]
 l_dyn_charm_3 [{shapeType=Capsule, length=1, radius=1, center=<1 0 0>, rotation=<rot 90 0 0>}]
 l_dyn_charm_4 [{shapeType=Capsule, length=1.09, radius=2.1824, center=<2.6358 0 0>, rotation=<rot 90 0 0>}]
 pack [{shapeType=Capsule, length=7.2466, radius=8.928, shapeName=pack_capsule, center=<0 5.3298 -23.3269>, rotation=<rot 90 0.0003 0>}]
 r_base_charm [{shapeType=Capsule, length=12.0227, radius=7.8311, center=<18.8748 2.6933 -12.9798>, rotation=<rot 0 0 90>}]
 r_dyn_charm_1 [{shapeType=Capsule, length=1, radius=1, center=<2.1002 0 0>, rotation=<rot 90 0 0>}]
 r_dyn_charm_2 [{shapeType=Capsule, length=1, radius=1, center=<1 0 0>, rotation=<rot 90 0 0>}]
 r_dyn_charm_3 [{shapeType=Capsule, length=1, radius=1, center=<1 0 0>, rotation=<rot 90 0 0>}]
 r_dyn_charm_4 [{shapeType=Capsule, length=0.9699, radius=1.6701, center=<1.8508 0 0>, rotation=<rot 90 0 0>}]
 root [{shapeType=Capsule, length=2, radius=2, shapeName=root_capsule, center=<0 -10.3077 -22.8153>}]
}
constraints [
 {bone1Name=c_dyn_charm_2, bone2Name=c_base_charm, swing1LimitDegrees=35, swing2LimitDegrees=35, twistLimitDegrees=2}
 {bone1Name=c_dyn_charm_3, bone2Name=c_dyn_charm_2, swing1LimitDegrees=35, swing2LimitDegrees=35, twistLimitDegrees=2}
 {
  bone1Name=l_base_charm
  bone2Name=pack
  swing1LimitDegrees=45
  swing1Motion=Free
  swing2LimitDegrees=45
  swing2Motion=Free
  twistLimitDegrees=45
  twistMotion=Free
 }
 {bone1Name=l_dyn_charm_1, bone2Name=l_base_charm, swing1LimitDegrees=35, swing2LimitDegrees=35, twistLimitDegrees=2}
 {bone1Name=l_dyn_charm_2, bone2Name=l_dyn_charm_1, swing1LimitDegrees=35, swing2LimitDegrees=35, twistLimitDegrees=2}
 {bone1Name=l_dyn_charm_3, bone2Name=l_dyn_charm_2, swing1LimitDegrees=35, swing2LimitDegrees=35, twistLimitDegrees=2}
 {bone1Name=l_dyn_charm_4, bone2Name=l_dyn_charm_3, swing1LimitDegrees=35, swing2LimitDegrees=35, twistLimitDegrees=2}
 {
  bone1Name=r_base_charm
  bone2Name=pack
  swing1LimitDegrees=45
  swing1Motion=Free
  swing2LimitDegrees=45
  swing2Motion=Free
  twistLimitDegrees=45
  twistMotion=Free
 }
 {bone1Name=r_dyn_charm_1, bone2Name=r_base_charm, swing1LimitDegrees=35, swing2LimitDegrees=35, twistLimitDegrees=2}
 {bone1Name=r_dyn_charm_2, bone2Name=r_dyn_charm_1, swing1LimitDegrees=35, swing2LimitDegrees=35, twistLimitDegrees=2}
 {bone1Name=r_dyn_charm_3, bone2Name=r_dyn_charm_2, swing1LimitDegrees=35, swing2LimitDegrees=35, twistLimitDegrees=2}
 {bone1Name=r_dyn_charm_4, bone2Name=r_dyn_charm_3, swing1LimitDegrees=35, swing2LimitDegrees=35, twistLimitDegrees=2}
 {bone1Name=c_base_charm, bone2Name=pack, swing1LimitDegrees=45, swing2LimitDegrees=45, twistLimitDegrees=2}
 {
  bone1Name=pack
  bone2Name=root
  swing1LimitDegrees=45
  swing1Motion=Free
  swing2LimitDegrees=45
  swing2Motion=Free
  twistLimitDegrees=45
  twistMotion=Free
 }
]
