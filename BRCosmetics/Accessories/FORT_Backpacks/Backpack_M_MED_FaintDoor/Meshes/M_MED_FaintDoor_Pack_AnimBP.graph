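# Animation Blueprint for the FaintDoor cosmetic backpack, adding posture-based offset and physics-driven secondary motion.
# Each update maps the owner's spine_01-to-spine_05 yaw deviation from reference pose to a root roll and Y shift, applied additively to the input pose.
# A RigidBody node then simulates the pack in "pack" bone space.
# Its gravity comes from a nativized gravity-override state function using the backpack's parameter asset.
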AnimBlueprint /BRCosmetics/Accessories/FORT_Backpacks/Backpack_M_MED_FaintDoor/Meshes/M_MED_FaintDoor_Pack_AnimBP
target_skeleton=&/M_MED_FaintDoor_Pack_Skeleton
graphs [
 AnimGraph {nodes [
   n1 = Input Pose()
   n2 = RigidBody!(
    ComponentPose="Transform (Modify) Bone - Bone: root"(
     ComponentPose=Local To Component(LocalPose=@n1)
     Translation=Get Backpack Translation() -> (Backpack Translation:Vector)
     Rotation=Get Backpack Rotate() -> (Backpack Rotate:Rotator)
    ) {translationMode=BMM_Additive, boneToModify {boneName=root}}
    Alpha
    OverrideWorldGravity:Vector=Get GravityOverride()
    ExternalForce
   ) -> (Pose) {
    bClampLinearTranslationLimitToRefPose=true
    bOverrideWorldGravity=true
    overridePhysicsAsset=&/M_MED_FaintDoor_Pack_Physics
    simulationSpace=BaseBoneSpace
    baseBoneRef {boneName=pack}
    simSpaceSettings {
     dampingAlpha=0
     maxAngularAcceleration=21
     maxAngularVelocity=8
     maxLinearAcceleration=1500
     maxLinearVelocity=700
     worldAlpha=1
     externalLinearDragV=<0.75 0.75 0.75>
    }
   }
   Output Pose(Result=Component To Local(ComponentPose=@n2)) {layerGroup=None}
  ]}
 EventGraph {nodes [
   Event BlueprintUpdateAnimation()
   > n1 = "Gravity Override State Function (Parameter Asset) (Nativized)"(
    CharacterPartAnimInstance=Self-Reference()
    GameTimeInSeconds=GetGameTimeInSeconds()
    Parameters=&/M_MED_FaintDoor_Pack_Parameters
   )
   > Set GravityOverride(GravityOverride=@n1)
   n2 = GetOwnerSkeletalMeshComponent()
   n3 = BreakRotator(InRot=BreakTransform(InTransform=GetDeltaTransformFromRefPose(self=@n2, BoneName=spine_05, BaseName=spine_01)).Rotation)
   > Set Backpack Rotate(Backpack Rotate:Rotator=MakeRotator(Roll=MapRangeClamped(Value=@n3.Yaw, InRangeA=0, InRangeB=-30, OutRangeA=5, OutRangeB=25))) -> (Output_Get:Rotator)
   > Set Backpack Translation(Backpack Translation:Vector=MakeVector(Y=MapRangeClamped(Value=@n3.Yaw, InRangeA=0, InRangeB=-30, OutRangeB=-6.5))) -> (Output_Get:Vector)
   TryGetPawnOwner()
   Set RBAN Settings Out()
  ]}
]
variables [
 Backpack Rotate {type=Rotator}
 Backpack Translation {type=Vector}
 GravityOverride {type=Vector}
 RBAN Settings Out {type=FortRigidBodyAnimNodeParameters}
]
